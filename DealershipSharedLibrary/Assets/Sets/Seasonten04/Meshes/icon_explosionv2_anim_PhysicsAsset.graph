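# Physics asset for the animated explosion icon skeletal mesh (icon_explosionv2_anim) in the Seasonten04 set of the dealership shared library.
# It holds a single capsule body on the root_jnt bone, slightly offset and rotated from the bone.
# There are no constraints, so the mesh has one simple collision shape rather than a jointed ragdoll setup.
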PhysicsAsset /DealershipSharedLibrary/Assets/Sets/Seasonten04/Meshes/icon_explosionv2_anim_PhysicsAsset
bodies {root_jnt [{shapeType=Capsule, length=10.0079, radius=9.8901, center=<-0.3597 -10.2321 -0.6827>, rotation=<rot 2.7172 67.6905 -90>}]}
constraints []
